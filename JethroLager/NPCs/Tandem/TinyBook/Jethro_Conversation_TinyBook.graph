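# Conversation graph for the TinyBook NPC, voiced as Jar Jar Binks, serving as its opening greeting.
# A single entry point leads to a speech node whose default line can be replaced by contextual lines for a first-ever meeting, per-match conversation count checks, or a specific character outfit.
# Priority curves rank the first-meeting and outfit reactions.
# The speech node then links out to shared quest, Jethro services, and TinyBook dialogue entry points.
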ConversationDatabase /JethroLager/NPCs/Tandem/TinyBook/Jethro_Conversation_TinyBook
entry_points [{tag=Jethro.Conversation.Entry.TinyBook, destinations [#1]}]
nodes [
 ConversationEntryPointNode_7 {
  id=#1
  class=ConversationEntryPointNode
  outputs [#2]
  properties {entryTag {tagName=Jethro.Conversation.Entry.TinyBook}}
 }
 Speech_C_4 {
  id=#2
  class=Speech_C
  outputs [#3, #4, #5]
  properties {
   nodeName=""
   generalConfig {
    defaultMessage="WHOO-HOO-HOO!"
    contextualMessages [
     {
      message="Meesa Jar Jar Binks! Weesa gonna be friends! Okieday?"
      contextRequirements [{requirement {class=FortControllerRequirement_IsFirstEverConversationWithNPC}}]
      priority {curve {rowName=Tandem.Speech.Priority.FirstEverInteraction}}
     }
     {
      message="WHOO-HOO-HOO!"
      contextRequirements [{requirement {
         class=FortControllerRequirement_CheckNPCMatchConversationCount
         properties {nPCUniqueIdTag {tagName=AISpawnerData.Type.Tandem.Jethro.TinyBook}}
        }}]
      priority {curve {rowName=None}}
     }
     {
      message=(Jar Jar blows a raspberry at you.)
      contextRequirements [{requirement {
         class=FortControllerRequirement_CheckNPCMatchConversationCount
         properties {nPCUniqueIdTag {tagName=AISpawnerData.Type.Tandem.Jethro.TinyBook}}
        }}]
      priority {curve {rowName=None}}
     }
     {
      message=(Jar Jar's devouring a wriggling frog.)
      contextRequirements [{requirement {
         class=FortControllerRequirement_CheckNPCMatchConversationCount
         properties {nPCUniqueIdTag {tagName=AISpawnerData.Type.Tandem.Jethro.TinyBook}}
        }}]
      priority {curve {rowName=None}}
     }
     {
      message="Annie?!"
      contextRequirements [{requirement {class=FortControllerRequirement_HasCID, properties {allowedCIds [/BRCosmetics/Athena/Items/Cosmetics/Characters/Character_PlotTwist]}}}]
      priority {curve {rowName=Tandem.Speech.Priority.OutfitReaction}}
     }
    ]
   }
  }
 }
 ConversationLinkNode_3 {
  id=#5
  class=ConversationLinkNode
  properties {remoteEntryTag {tagName=Jethro.Conversation.Entry.TinyBook.Dialogue}}
 }
 ConversationLinkNode_0 {id=#4, class=ConversationLinkNode, properties {remoteEntryTag {tagName=Conversation.Entry.JethroServices}}}
 ConversationLinkNode_1 {id=#3, class=ConversationLinkNode, properties {remoteEntryTag {tagName=Conversation.Entry.NPCLibrary.Quests}}}
]
